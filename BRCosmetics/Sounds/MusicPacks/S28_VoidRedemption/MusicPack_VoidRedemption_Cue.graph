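# Sound cue for the Season 28 Void Redemption music pack, playing a one-shot intro into a looping track.
# A mixer starts the intro immediately and the looping wave after a fixed delay of about 1.8 seconds, under a single-entry random root.
# It uses the music pack sound class at slightly reduced volume and has a send to the music audio-analysis submix.
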SoundCue /BRCosmetics/Sounds/MusicPacks/S28_VoidRedemption/MusicPack_VoidRedemption_Cue
root=Random_0
nodes [
 {id=Delay_0, children [WavePlayer_0], properties {delayMax=1.777, delayMin=1.777}}
 {id=Mixer_0, children [Delay_0,WavePlayer_1], properties {inputVolume [1,1]}}
 {id=Random_0, children [Mixer_0], properties {weights [1]}}
 {
  id=WavePlayer_0
  properties {bLooping=true, soundWaveAssetPtr=/Game/Athena/Sounds/MusicPacks/S28_VoidRedemption/S28_MusicPack_VoidRedemption_Loop}
 }
 {id=WavePlayer_1, properties {soundWaveAssetPtr=/Game/Athena/Sounds/MusicPacks/S28_VoidRedemption/S28_MusicPack_VoidRedemption_Intro}}
]
properties {
 maxDistance=2097152
 soundClassObject=/Game/Sounds/FortSoundClassesAndModes/Fort_Music_MusicPacks
 volumeMultiplier=0.8
 soundSubmixSends [{minSendLevel=0, soundSubmix=&Fortnite_AudioAnalysis_Music}]
}
